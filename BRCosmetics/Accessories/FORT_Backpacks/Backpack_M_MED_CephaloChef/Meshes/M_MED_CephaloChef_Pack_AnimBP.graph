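# Animation Blueprint for the CephaloChef backpack, adding procedural secondary motion to its flag, flagstaff, ropes and fish.
# Additive pack rotations feed a cached pose, then wind-enabled AnimDynamics flag chains are blended by bIsInFrontEnd.
# Two rope setups follow, blended by bIsSkyDiving to swap normal gravity for an upward, sideways pull, before linear-spring fish and a tightly limited flagstaff.
# The update event stores the inverse skydiving state in IsNOTSkyDiving.
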AnimBlueprint /BRCosmetics/Accessories/FORT_Backpacks/Backpack_M_MED_CephaloChef/Meshes/M_MED_CephaloChef_Pack_AnimBP
target_skeleton=&/M_MED_CephaloChef_Pack_Skeleton
graphs [
 AnimGraph {nodes [
   n1 = Anim Dynamics - Bone: dyn_rope(ComponentPose=Local To Component(LocalPose=Use cached pose 'FrontendPose'()), Alpha=1, ExternalForce="0,0,0") {
    angularDampingOverride=0.9999
    angularSpringConstant=10000
    bAngularSpring=true
    bOverrideAngularDamping=true
    bUseGravityOverride=true
    simulationSpace=RootRelative
    alphaBoolBlend {blendOption=Linear}
    boundBone {boneName=dyn_rope}
    componentAppliedLinearAccClamp=<1000 1000 1000>
    componentLinearAccScale=<1 1 1>
    componentLinearVelScale=<1 1 1>
    gravityOverride {y=-300, z=500}
    physicsBodyDefinitions [{
      boundBone {boneName=dyn_rope}
      boxExtents=<30 20 30>
      constraintSetup {angularLimitsMax=<5 15 20>, angularLimitsMin=<-5 -15 0>, angularTarget=<1 0 0>}
      localJointOffset=<1 0 0>
     }]
   }
   n2 = Anim Dynamics - Bone: dyn_rope_l(ComponentPose=@n1, Alpha=1, ExternalForce="0,0,0") {
    angularDampingOverride=1
    angularSpringConstant=2000
    bAngularSpring=true
    bOverrideAngularDamping=true
    bUseGravityOverride=true
    simulationSpace=RootRelative
    alphaBoolBlend {blendOption=Linear}
    boundBone {boneName=dyn_rope_l}
    componentAppliedLinearAccClamp=<1000 1000 1000>
    componentLinearAccScale=<0.5 0.5 0.5>
    componentLinearVelScale=<0.7 0.7 0.7>
    gravityOverride {y=-300, z=500}
    physicsBodyDefinitions [{
      boundBone {boneName=dyn_rope_l}
      boxExtents=<12 10 10>
      constraintSetup {angularLimitsMax=<0 40 40>, angularLimitsMin=<0 -40 -40>, angularTarget=<1 0 0>}
      localJointOffset=<1 0 0>
     }]
   }
   n3 = Anim Dynamics - Bone: dyn_rope_r(ComponentPose=@n2, Alpha=1, ExternalForce="0,0,0") {
    angularDampingOverride=0.9999
    angularSpringConstant=1500
    bAngularSpring=true
    bOverrideAngularDamping=true
    bUseGravityOverride=true
    simulationSpace=RootRelative
    alphaBoolBlend {blendOption=Linear}
    boundBone {boneName=dyn_rope_r}
    componentAppliedLinearAccClamp=<1200 1200 1200>
    componentLinearAccScale=<0.5 0.5 0.5>
    componentLinearVelScale=<0.7 0.7 0.7>
    gravityOverride {y=-300, z=500}
    physicsBodyDefinitions [{
      boundBone {boneName=dyn_rope_r}
      boxExtents=<10 10 10>
      constraintSetup {angularLimitsMax=<0 40 40>, angularLimitsMin=<0 -40 -40>, angularTarget=<1 0 0>}
      localJointOffset=<0.8 0 0>
     }]
   }
   n4 = Component To Local(ComponentPose=@n3)
   n5 = Anim Dynamics - Bone: dyn_rope(ComponentPose=Local To Component(LocalPose=Use cached pose 'FrontendPose'()), Alpha=1, ExternalForce="0,0,0") {
    angularDampingOverride=0.9999
    angularSpringConstant=10000
    bAngularSpring=true
    bOverrideAngularDamping=true
    bUseGravityOverride=true
    simulationSpace=RootRelative
    alphaBoolBlend {blendOption=Linear}
    boundBone {boneName=dyn_rope}
    componentAppliedLinearAccClamp=<1000 1000 1000>
    componentLinearAccScale=<1 1 1>
    componentLinearVelScale=<1 1 1>
    gravityOverride {z=-980}
    physicsBodyDefinitions [{
      boundBone {boneName=dyn_rope}
      boxExtents=<30 20 30>
      constraintSetup {angularLimitsMax=<5 15 20>, angularLimitsMin=<-5 -15 0>, angularTarget=<1 0 0>}
      localJointOffset=<1 0 0>
     }]
   }
   n6 = Anim Dynamics - Bone: dyn_rope_l(ComponentPose=@n5, Alpha=1, ExternalForce="0,0,0") {
    angularDampingOverride=1
    angularSpringConstant=2000
    bAngularSpring=true
    bOverrideAngularDamping=true
    bUseGravityOverride=true
    simulationSpace=RootRelative
    alphaBoolBlend {blendOption=Linear}
    boundBone {boneName=dyn_rope_l}
    componentAppliedLinearAccClamp=<1000 1000 1000>
    componentLinearAccScale=<0.5 0.5 0.5>
    componentLinearVelScale=<0.7 0.7 0.7>
    gravityOverride {z=-980}
    physicsBodyDefinitions [{
      boundBone {boneName=dyn_rope_l}
      boxExtents=<12 10 10>
      constraintSetup {angularLimitsMax=<0 40 40>, angularLimitsMin=<0 -40 -40>, angularTarget=<1 0 0>}
      localJointOffset=<1 0 0>
     }]
   }
   n7 = Anim Dynamics - Bone: dyn_rope_r(ComponentPose=@n6, Alpha=1, ExternalForce="0,0,0") {
    angularDampingOverride=0.9999
    angularSpringConstant=1500
    bAngularSpring=true
    bOverrideAngularDamping=true
    bUseGravityOverride=true
    simulationSpace=RootRelative
    alphaBoolBlend {blendOption=Linear}
    boundBone {boneName=dyn_rope_r}
    componentAppliedLinearAccClamp=<1200 1200 1200>
    componentLinearAccScale=<0.5 0.5 0.5>
    componentLinearVelScale=<0.7 0.7 0.7>
    gravityOverride {z=-980}
    physicsBodyDefinitions [{
      boundBone {boneName=dyn_rope_r}
      boxExtents=<10 10 10>
      constraintSetup {angularLimitsMax=<0 40 40>, angularLimitsMin=<0 -40 -40>, angularTarget=<1 0 0>}
      localJointOffset=<0.8 0 0>
     }]
   }
   n8 = Component To Local(ComponentPose=@n7)
   n9 = Get bIsSkyDiving()
   n10 = Anim Dynamics - Bone: dyn_fish_01!(
    ComponentPose=Local To Component(LocalPose=Blend Poses by bool(BlendPose_0=@n4, BlendPose_1=@n8, bActiveValue=@n9) {blendType=Linear})
    bAlphaBoolEnabled:bool=True
    ExternalForce="0,0,0"
   ) -> (Pose) {
    alphaInputType=Bool
    bLinearSpring=true
    bOverrideLinearDamping=true
    bUseGravityOverride=true
    linearDampingOverride=0.99
    linearSpringConstant=100
    simulationSpace=RootRelative
    alphaBoolBlend {blendOption=Linear}
    boundBone {boneName=dyn_fish_01}
    componentAppliedLinearAccClamp=<0 0 1000>
    componentLinearAccScale {z=0.3}
    componentLinearVelScale {z=0.5}
    gravityOverride {z=-980}
    physicsBodyDefinitions [{
      boundBone {boneName=dyn_fish_01}
      boxExtents=<10 15 15>
      constraintSetup {angularLimitsMax=<0 0 0>, angularLimitsMin=<0 0 0>, angularTarget=<1 0 0>, linearAxesMax=<5 0 0>}
      localJointOffset=<0 0 0>
     }]
   }
   n11 = Anim Dynamics - Bone: dyn_fish_02!(ComponentPose=@n10, bAlphaBoolEnabled:bool=True, ExternalForce="0,0,0") -> (Pose) {
    alphaInputType=Bool
    bLinearSpring=true
    bOverrideLinearDamping=true
    bUseGravityOverride=true
    linearDampingOverride=0.99
    linearSpringConstant=200
    simulationSpace=RootRelative
    alphaBoolBlend {blendOption=Linear}
    boundBone {boneName=dyn_fish_02}
    componentAppliedLinearAccClamp=<0 0 1100>
    componentLinearAccScale {z=0.4}
    componentLinearVelScale {z=0.4}
    gravityOverride {z=-980}
    physicsBodyDefinitions [{
      boundBone {boneName=dyn_fish_02}
      boxExtents=<10 10 10>
      constraintSetup {angularLimitsMax=<0 0 0>, angularLimitsMin=<0 0 0>, angularTarget=<1 0 0>, linearAxesMax=<5 0 0>}
      localJointOffset=<0 0 0>
     }]
   }
   n12 = Anim Dynamics - Bone: dyn_fish_03!(ComponentPose=@n11, bAlphaBoolEnabled:bool=True, ExternalForce="0,0,0") -> (Pose) {
    alphaInputType=Bool
    bLinearSpring=true
    bOverrideLinearDamping=true
    bUseGravityOverride=true
    linearDampingOverride=0.99
    linearSpringConstant=150
    simulationSpace=RootRelative
    alphaBoolBlend {blendOption=Linear}
    boundBone {boneName=dyn_fish_03}
    componentAppliedLinearAccClamp=<0 0 1000>
    componentLinearAccScale {z=0.3}
    componentLinearVelScale {z=0.5}
    gravityOverride {z=-950}
    physicsBodyDefinitions [{
      boundBone {boneName=dyn_fish_03}
      boxExtents=<10 15 15>
      constraintSetup {angularLimitsMax=<0 0 0>, angularLimitsMin=<0 0 0>, angularTarget=<1 0 0>, linearAxesMax=<5 0 0>}
      localJointOffset=<0 0 0>
     }]
   }
   n13 = Anim Dynamics - Bone: dyn_flagstaff(ComponentPose=@n12, Alpha=1, ExternalForce="0,0,0") {
    angularDampingOverride=0.9999
    angularSpringConstant=10000
    bAngularSpring=true
    bOverrideAngularDamping=true
    bUseGravityOverride=true
    simulationSpace=RootRelative
    alphaBoolBlend {blendOption=Linear}
    boundBone {boneName=dyn_flagstaff}
    componentAppliedLinearAccClamp=<1000 1000 0>
    componentLinearAccScale=<0.7 0.7>
    componentLinearVelScale=<0.8 0.8>
    gravityOverride {z=-980}
    physicsBodyDefinitions [{
      boundBone {boneName=dyn_flagstaff}
      boxExtents=<50 20 20>
      constraintSetup {angularLimitsMax=<0 2.5 2.5>, angularLimitsMin=<0 -2.5 -2.5>, angularTarget=<1 0 0>}
      localJointOffset=<1 0 0>
     }]
   }
   Output Pose(Result=Component To Local(ComponentPose=@n13)) {layerGroup=None}
   n14 = "Transform (Modify) Bone - Bone: pack"!(ComponentPose=Local To Component(LocalPose=Input Pose()), bAlphaBoolEnabled:bool, Rotation="0,0,5") -> (Pose) {
    alphaInputType=Bool
    rotationMode=BMM_Additive
    alphaBoolBlend {blendInTime=0.2, blendOutTime=0.2}
    boneToModify {boneName=pack}
    rotation {roll=5}
   }
   n15 = "Transform (Modify) Bone - Bone: pack"(ComponentPose=@n14, Rotation="0,0,-10") {
    alphaInputType=Bool
    rotationMode=BMM_Additive
    alphaBoolBlend {blendInTime=0.8, blendOutTime=0.3}
    boneToModify {boneName=pack}
    rotation {roll=-10}
   }
   Save cached pose 'PreSimPose'(Pose=Component To Local(ComponentPose=@n15)) {cacheName=PreSimPose}
   n16 = Anim Dynamics - Bone: dyn_flag_01(ComponentPose=Local To Component(LocalPose=Use cached pose 'PreSimPose'()), Alpha=1, ExternalForce="0,0,0") {
    angularDampingOverride=1
    angularSpringConstant=500
    bAngularSpring=true
    bEnableWind=true
    bOverrideAngularDamping=true
    bOverrideLinearDamping=true
    bUseGravityOverride=true
    gravityScale=0.56
    linearDampingOverride=1
    simulationSpace=RootRelative
    windScale=1000
    alphaBoolBlend {blendOption=Linear}
    boundBone {boneName=dyn_flag_01}
    componentAppliedLinearAccClamp=<300 300 300>
    componentLinearAccScale=<0.3 0.3 0.3>
    componentLinearVelScale=<0.4 0.4 0.4>
    gravityOverride {z=-50}
    physicsBodyDefinitions [{
      boundBone {boneName=dyn_flag_01}
      boxExtents=<2.5 3 3>
      constraintSetup {angularLimitsMax=<0 10 0>, angularLimitsMin=<0 -7 0>, angularTarget=<1 0 0>}
      localJointOffset=<2 2 0.4>
     }]
   }
   n17 = Anim Dynamics - Bone: dyn_flag_02(ComponentPose=@n16, Alpha=1, ExternalForce="0,0,0") {
    angularDampingOverride=1
    angularSpringConstant=500
    bAngularSpring=true
    bEnableWind=true
    bOverrideAngularDamping=true
    bOverrideLinearDamping=true
    bUseGravityOverride=true
    gravityScale=0.56
    linearDampingOverride=1
    simulationSpace=RootRelative
    windScale=1000
    alphaBoolBlend {blendOption=Linear}
    boundBone {boneName=dyn_flag_02}
    componentAppliedLinearAccClamp=<300 300 300>
    componentLinearAccScale=<0.3 0.3 0.3>
    componentLinearVelScale=<0.4 0.4 0.4>
    gravityOverride {z=-50}
    physicsBodyDefinitions [{
      boundBone {boneName=dyn_flag_02}
      boxExtents=<3.5 3.5 3.5>
      constraintSetup {angularLimitsMax=<0 14 0>, angularLimitsMin=<0 -10 0>, angularTarget=<1 0 0>}
      localJointOffset=<2 1.5 0.4>
     }]
   }
   n18 = Anim Dynamics - Bone: dyn_flag_03(ComponentPose=@n17, Alpha=1, ExternalForce="0,0,0") {
    angularDampingOverride=1
    angularSpringConstant=500
    bAngularSpring=true
    bEnableWind=true
    bOverrideAngularDamping=true
    bOverrideLinearDamping=true
    bUseGravityOverride=true
    gravityScale=0.56
    linearDampingOverride=1
    simulationSpace=RootRelative
    windScale=1000
    alphaBoolBlend {blendOption=Linear}
    boundBone {boneName=dyn_flag_03}
    componentAppliedLinearAccClamp=<300 300 300>
    componentLinearAccScale=<0.3 0.3 0.3>
    componentLinearVelScale=<0.4 0.4 0.4>
    gravityOverride {z=-50}
    physicsBodyDefinitions [{
      boundBone {boneName=dyn_flag_03}
      boxExtents=<4.5 4.5 4.5>
      constraintSetup {angularLimitsMax=<1.5 20 0>, angularLimitsMin=<-1.5 -15 0>, angularTarget=<1 0 0>}
      localJointOffset=<2 0.65 0.4>
     }]
   }
   n19 = Anim Dynamics - Bone: dyn_flag_04(ComponentPose=@n18, Alpha=1, ExternalForce="0,0,0") {
    angularDampingOverride=1
    angularSpringConstant=500
    bAngularSpring=true
    bEnableWind=true
    bOverrideAngularDamping=true
    bOverrideLinearDamping=true
    bUseGravityOverride=true
    gravityScale=0.56
    linearDampingOverride=1
    simulationSpace=RootRelative
    windScale=1000
    alphaBoolBlend {blendOption=Linear}
    boundBone {boneName=dyn_flag_04}
    componentAppliedLinearAccClamp=<300 300 300>
    componentLinearAccScale=<0.3 0.3 0.3>
    componentLinearVelScale=<0.4 0.4 0.4>
    gravityOverride {z=-50}
    physicsBodyDefinitions [{
      boundBone {boneName=dyn_flag_04}
      boxExtents=<7.5 7.5 7.5>
      constraintSetup {angularLimitsMax=<1.5 30 0>, angularLimitsMin=<-1.5 -20 0>, angularTarget=<1 0 0>}
      localJointOffset=<2 -0.2 0.4>
     }]
   }
   n20 = Anim Dynamics - Bone: dyn_flag_01(ComponentPose=Local To Component(LocalPose=Use cached pose 'PreSimPose'()), Alpha=1, ExternalForce="0,0,0") {
    angularDampingOverride=1
    angularSpringConstant=100
    bAngularSpring=true
    bEnableWind=true
    bOverrideAngularDamping=true
    bOverrideLinearDamping=true
    bUseGravityOverride=true
    gravityScale=0.56
    linearDampingOverride=1
    simulationSpace=RootRelative
    windScale=1000
    alphaBoolBlend {blendOption=Linear}
    boundBone {boneName=dyn_flag_01}
    componentAppliedLinearAccClamp=<300 300 300>
    componentLinearAccScale=<0.3 0.3 0.3>
    componentLinearVelScale=<0.4 0.4 0.4>
    gravityOverride {z=-500}
    physicsBodyDefinitions [{
      boundBone {boneName=dyn_flag_01}
      boxExtents=<2.5 3 3>
      constraintSetup {angularLimitsMax=<0 10 0>, angularLimitsMin=<0 -7 0>, angularTarget=<1 0 0>}
      localJointOffset=<2 0 0>
     }]
   }
   n21 = Anim Dynamics - Bone: dyn_flag_02(ComponentPose=@n20, Alpha=1, ExternalForce="0,0,0") {
    angularDampingOverride=1
    angularSpringConstant=100
    bAngularSpring=true
    bEnableWind=true
    bOverrideAngularDamping=true
    bOverrideLinearDamping=true
    bUseGravityOverride=true
    gravityScale=0.56
    linearDampingOverride=1
    simulationSpace=RootRelative
    windScale=1000
    alphaBoolBlend {blendOption=Linear}
    boundBone {boneName=dyn_flag_02}
    componentAppliedLinearAccClamp=<300 300 300>
    componentLinearAccScale=<0.4 0.4 0.4>
    componentLinearVelScale=<0.4 0.4 0.4>
    gravityOverride {z=-500}
    physicsBodyDefinitions [{
      boundBone {boneName=dyn_flag_02}
      boxExtents=<3 3 3>
      constraintSetup {angularLimitsMax=<0 14 0>, angularLimitsMin=<0 -10 0>, angularTarget=<1 0 0>}
      localJointOffset=<2 0 0>
     }]
   }
   n22 = Anim Dynamics - Bone: dyn_flag_03(ComponentPose=@n21, Alpha=1, ExternalForce="0,0,0") {
    angularDampingOverride=1
    angularSpringConstant=100
    bAngularSpring=true
    bEnableWind=true
    bOverrideAngularDamping=true
    bOverrideLinearDamping=true
    bUseGravityOverride=true
    gravityScale=0.56
    linearDampingOverride=1
    simulationSpace=RootRelative
    windScale=1000
    alphaBoolBlend {blendOption=Linear}
    boundBone {boneName=dyn_flag_03}
    componentAppliedLinearAccClamp=<300 300 300>
    componentLinearAccScale=<0.4 0.4 0.4>
    componentLinearVelScale=<0.4 0.4 0.4>
    gravityOverride {z=-500}
    physicsBodyDefinitions [{
      boundBone {boneName=dyn_flag_03}
      boxExtents=<5 5 5>
      constraintSetup {angularLimitsMax=<1.5 20 0>, angularLimitsMin=<-1.5 -15 0>, angularTarget=<1 0 0>}
      localJointOffset=<3 0 0>
     }]
   }
   n23 = Anim Dynamics - Bone: dyn_flag_04(ComponentPose=@n22, Alpha=1, ExternalForce="0,0,0") {
    angularDampingOverride=1
    angularSpringConstant=500
    bAngularSpring=true
    bEnableWind=true
    bOverrideAngularDamping=true
    bOverrideLinearDamping=true
    bUseGravityOverride=true
    gravityScale=0.56
    linearDampingOverride=1
    simulationSpace=RootRelative
    windScale=1000
    alphaBoolBlend {blendOption=Linear}
    boundBone {boneName=dyn_flag_04}
    componentAppliedLinearAccClamp=<300 300 300>
    componentLinearAccScale=<0.3 0.3 0.3>
    componentLinearVelScale=<0.4 0.4 0.4>
    physicsBodyDefinitions [{
      boundBone {boneName=dyn_flag_04}
      boxExtents=<7.5 7.5 7.5>
      constraintSetup {angularLimitsMax=<1.5 30 0>, angularLimitsMin=<-1.5 -20 0>, angularTarget=<1 0 0>}
      localJointOffset=<2 0 0>
     }]
   }
   Save cached pose 'FrontendPose'(Pose=Blend Poses by bool(BlendPose_0=Component To Local(ComponentPose=@n19), BlendPose_1=Component To Local(ComponentPose=@n23), bActiveValue=Get bIsInFrontEnd()) {blendType=Linear}) {cacheName=FrontendPose}
  ]}
 EventGraph {nodes [
   Event BlueprintUpdateAnimation()
   > Set IsNOTSkyDiving(IsNOTSkyDiving:bool=NOT Boolean(A=Get bIsSkyDiving())) -> (Output_Get:bool)
   TryGetPawnOwner()
  ]}
]
variables [IsNOTSkyDiving {type=boolean}]
